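# Physics asset for the JailBirds backpack skeletal mesh, defining simulation bodies for its dangling accessories and flaps.
# Each bone gets one capsule, and items like the flashlight, crystal, gem, crowbar, cat and map are constrained directly to backpack_root.
# Left and right straps chain off their flaps with tight swing and twist limits, while three hook bodies form a looser chain from the root.
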PhysicsAsset /Game/Accessories/FORT_Backpacks/M_MED_JailBirds_Backpack/Meshes/SK_M_MED_JailBirds_Backpack_Physics
bodies {
 backpack_root [{shapeType=Capsule, length=2, radius=2}]
 cat [{shapeType=Capsule, length=7.8055, radius=3.6026, center=<-3.382 0 0.0004>, rotation=<rot 90 0 0>}]
 crowbar [{shapeType=Capsule, length=13.9382, radius=1.1506, center=<-3.1629 0 0>, rotation=<rot 90 0 0>}]
 crystal [{shapeType=Capsule, length=5.9169, radius=1.7879, center=<0.5968 0 0>, rotation=<rot 90 0 0>}]
 flap_l [{shapeType=Capsule, length=2.4776, radius=2.9407, center=<0.8142 0 0>, rotation=<rot 90 0 0>}]
 flap_r [{shapeType=Capsule, length=3.2743, radius=3, center=<0.788 0.0002 0>, rotation=<rot 90 0 0>}]
 flashlight [{shapeType=Capsule, length=12.8513, radius=2.7276, center=<-2.3582 -0.0384 -5.3698>, rotation=<rot 0.0843 -89.9983 1.0505>}]
 gem [{shapeType=Capsule, length=2.0958, radius=2.3337, center=<0.9088 -0.0834 -0.0258>, rotation=<rot 84.9078 68.4674 158.5448>}]
 hook_01 [{shapeType=Capsule, length=1.75, radius=2.75, center=<-1 0 0>, rotation=<rot -90 0 0>}]
 hook_02 [{shapeType=Capsule, length=3.25, radius=2.25, center=<-1 0 0>, rotation=<rot -90 0 0>}]
 hook_03 [{shapeType=Capsule, length=6, radius=1.25, center=<-3.1405 0 0>, rotation=<rot -90 0 0>}]
 map [{shapeType=Capsule, length=18.8893, radius=3.1173, center=<-2.8869 0 0>, rotation=<rot 90 0 0>}]
 strap_l [{shapeType=Capsule, length=2.0066, radius=2.2299, center=<0.5726 0 0.0001>, rotation=<rot 90 0 0>}]
 strap_r [{shapeType=Capsule, length=2.6563, radius=1.5295, center=<0.1972 0 0>, rotation=<rot 90 0 0>}]
}
constraints [
 {bone1Name=flashlight, bone2Name=backpack_root, swing1LimitDegrees=45, swing2LimitDegrees=45, twistLimitDegrees=45}
 {bone1Name=crystal, bone2Name=backpack_root, swing1LimitDegrees=45, swing2LimitDegrees=45, twistLimitDegrees=2}
 {bone1Name=flap_l, bone2Name=backpack_root, swing1LimitDegrees=5, swing2LimitDegrees=45, twistLimitDegrees=2}
 {bone1Name=strap_l, bone2Name=flap_l, swing1LimitDegrees=15, swing2LimitDegrees=45, twistLimitDegrees=2}
 {bone1Name=flap_r, bone2Name=backpack_root, swing1LimitDegrees=5, swing2LimitDegrees=45, twistLimitDegrees=2}
 {bone1Name=strap_r, bone2Name=flap_r, swing1LimitDegrees=15, swing2LimitDegrees=45, twistLimitDegrees=2}
 {bone1Name=gem, bone2Name=backpack_root, swing1LimitDegrees=45, swing2LimitDegrees=45, twistLimitDegrees=5}
 {bone1Name=crowbar, bone2Name=backpack_root, swing1LimitDegrees=45, swing2LimitDegrees=45, twistLimitDegrees=5}
 {bone1Name=cat, bone2Name=backpack_root, swing1LimitDegrees=45, swing2LimitDegrees=45, twistLimitDegrees=5}
 {bone1Name=map, bone2Name=backpack_root, swing1LimitDegrees=45, swing2LimitDegrees=45, twistLimitDegrees=5}
 {bone1Name=hook_01, bone2Name=backpack_root, swing1LimitDegrees=90, swing2LimitDegrees=90, twistLimitDegrees=30}
 {bone1Name=hook_02, bone2Name=hook_01, swing1LimitDegrees=90, swing2LimitDegrees=90, twistLimitDegrees=30}
 {bone1Name=hook_03, bone2Name=hook_02, swing1LimitDegrees=90, swing2LimitDegrees=90, twistLimitDegrees=30}
]
