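# Animation Blueprint for the JailBirds backpack cosmetic, adding secondary physics motion to its dangling props, flaps, straps and grappling hook.
# A serial AnimDynamics pass springs each prop, flap and strap bone, then caches the pose.
# Two hook-chain simulations branch from that cache, and skydive, dive or float flags select the weaker, tighter-limited one.
# Trail, Apply Limits and RigidBody nodes are disconnected from the output, and the event graph is empty.
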AnimBlueprint /BRCosmetics/Accessories/FORT_Backpacks/M_MED_JailBirds_Backpack/Meshes/SK_M_MED_JailBirds_Backpack_AnimBP
target_skeleton=&/SK_M_MED_JailBirds_Backpack_Skeleton
graphs [
 AnimGraph {nodes [
   n1 = Trail controller - Bone: hook_03(Alpha=1) {
    bInvertChainBoneAxis=true
    chainLength=3
    rotationLimits [{limitMax=<180 180 180>}, {limitMax=<180 180 180>}, {limitMax=<180 180 180>}]
    rotationOffsets [<0 0 0>, <0 0 0>, <0 0 0>]
    trailBone {boneName=hook_03}
    trailRelaxationSpeed {editorCurveData {keys [<at 0 13 RCIM_Linear>,<at 1 10 RCIM_Linear>]}}
   }
   Apply Limits!(ComponentPose:Component Space Pose Link Structure=@n1, Alpha:float=1) -> (Pose:Component Space Pose Link Structure) {
    is=title AnimGraphNode_ApplyLimits
    alphaBoolBlend {blendOption=Linear}
    angularOffsets [<0 0 0>]
    angularRangeLimits [{bone {boneName=hook_01}, limitMax=<0 180 180>, limitMin=<-180 -180 -180>}]
   }
   RigidBody(Alpha=0.35) {
    lODThreshold=1
    overridePhysicsAsset=/Game/Accessories/FORT_Backpacks/M_MED_JailBirds_Backpack/Meshes/SK_M_MED_JailBirds_Backpack_Physics
    alphaBoolBlend {blendOption=Linear}
    componentLinearAccScale=<3 3 3>
    componentLinearVelScale=<8 8 8>
   }
   n2 = "Anim Dynamics - Chain: hook_01 -> hook_03"(ComponentPose=Local To Component(LocalPose=Use cached pose 'GrappelingHookCache'()), Alpha=0.4, ExternalForce="0,0,0") {
    angularDampingOverride=0.99
    angularSpringConstant=500
    bAngularSpring=true
    bChain=true
    bGravityOverrideInSimSpace=true
    bOverrideAngularDamping=true
    numSolverIterationsPostUpdate=4
    numSolverIterationsPreUpdate=16
    alphaBoolBlend {blendOption=Linear}
    boundBone {boneName=hook_01}
    chainEnd {boneName=hook_03}
    componentLinearAccScale=<0.8 0.8 0.8>
    componentLinearVelScale=<0.8 0.8 0.8>
    physicsBodyDefinitions [{
      sphereCollisionRadius=0
      boundBone {boneName=hook_01}
      boxExtents=<10 20 20>
      constraintSetup {angularLimitsMax=<5 10 60>, angularLimitsMin=<-5 -60 -15>, angularTarget=<1 0 0>}
      localJointOffset=<-2 0 0>
     }]
    relativeSpaceBone {boneName=backpack_root}
   }
   n3 = "Anim Dynamics - Chain: hook_01 -> hook_03"(ComponentPose=Local To Component(LocalPose=Use cached pose 'GrappelingHookCache'()), Alpha=0.6, ExternalForce="0,0,0") {
    angularDampingOverride=0.99
    angularSpringConstant=500
    bAngularSpring=true
    bChain=true
    bGravityOverrideInSimSpace=true
    bOverrideAngularDamping=true
    numSolverIterationsPostUpdate=4
    numSolverIterationsPreUpdate=16
    alphaBoolBlend {blendOption=Linear}
    boundBone {boneName=hook_01}
    chainEnd {boneName=hook_03}
    componentLinearAccScale=<0.8 0.8 0.8>
    componentLinearVelScale=<0.8 0.8 0.8>
    physicsBodyDefinitions [{
      sphereCollisionRadius=0
      boundBone {boneName=hook_01}
      boxExtents=<10 20 20>
      constraintSetup {angularLimitsMax=<5 15 60>, angularLimitsMin=<-5 -60 -15>, angularTarget=<1 0 0>}
      localJointOffset=<-2 0 0>
     }]
    relativeSpaceBone {boneName=backpack_root}
   }
   n4 = Get bIsSkydiveDiveMode()
   n5 = Get bIsSkydiveFloating()
   n6 = Get bIsSkyDiving()
   Output Pose(Result=Blend Poses by bool(
     BlendPose_0=Component To Local(ComponentPose=@n2)
     BlendPose_1=Component To Local(ComponentPose=@n3)
     bActiveValue=OR Boolean!(A=@n4, B=@n5, C:bool=@n6) -> (ReturnValue)
    ) {blendType=Linear}) {layerGroup=None}
   AnimGraphNode_SubInput_0 = Input Pose()
   n7 = Anim Dynamics - Bone: flashlight(ComponentPose=Local To Component(LocalPose=@AnimGraphNode_SubInput_0), Alpha=0.35, ExternalForce="0,0,0") {
    angularDampingOverride=0.99
    angularSpringConstant=5000
    bAngularSpring=true
    bGravityOverrideInSimSpace=true
    bOverrideAngularDamping=true
    numSolverIterationsPostUpdate=4
    numSolverIterationsPreUpdate=16
    alphaBoolBlend {blendOption=Linear}
    boundBone {boneName=flashlight}
    componentLinearAccScale=<0.8 0.8 0.8>
    componentLinearVelScale=<0.8 0.8 0.8>
    physicsBodyDefinitions [{
      sphereCollisionRadius=0
      boundBone {boneName=flashlight}
      boxExtents=<10 20 20>
      constraintSetup {angularLimitsMax=<5 20 20>, angularLimitsMin=<-5 -20 -15>, angularTarget=<1 0 0>}
      localJointOffset=<-3 0 0>
     }]
   }
   n8 = Anim Dynamics - Bone: map(ComponentPose=@n7, Alpha=0.2, ExternalForce="0,0,0") {
    angularDampingOverride=0.99
    angularSpringConstant=500
    bAngularSpring=true
    bGravityOverrideInSimSpace=true
    bOverrideAngularDamping=true
    numSolverIterationsPostUpdate=4
    numSolverIterationsPreUpdate=16
    alphaBoolBlend {blendOption=Linear}
    boundBone {boneName=map}
    componentLinearAccScale=<0.8 0.8 0.8>
    componentLinearVelScale=<0.8 0.8 0.8>
    physicsBodyDefinitions [{
      sphereCollisionRadius=0
      boundBone {boneName=map}
      boxExtents=<15 18 18>
      constraintSetup {angularLimitsMax=<5 20 20>, angularLimitsMin=<-5 -20 -20>, angularTarget=<1 0 0>}
      localJointOffset=<-3 0 0>
     }]
   }
   n9 = Anim Dynamics - Bone: cat(ComponentPose=@n8, Alpha=0.18, ExternalForce="0,0,0") {
    angularDampingOverride=0.99
    angularSpringConstant=500
    bAngularSpring=true
    bGravityOverrideInSimSpace=true
    bOverrideAngularDamping=true
    numSolverIterationsPostUpdate=4
    numSolverIterationsPreUpdate=16
    alphaBoolBlend {blendOption=Linear}
    boundBone {boneName=cat}
    componentLinearAccScale=<0.8 0.8 0.8>
    componentLinearVelScale=<0.8 0.8 0.8>
    physicsBodyDefinitions [{
      sphereCollisionRadius=0
      boundBone {boneName=cat}
      boxExtents=<8 15 15>
      constraintSetup {angularLimitsMax=<5 20 20>, angularLimitsMin=<-5 -20 -20>, angularTarget=<1 0 0>}
      localJointOffset=<-1.8 0 0>
     }]
   }
   n10 = Anim Dynamics - Bone: crowbar(ComponentPose=@n9, Alpha=0.22, ExternalForce="0,0,0") {
    angularDampingOverride=0.99
    angularSpringConstant=500
    bAngularSpring=true
    bGravityOverrideInSimSpace=true
    bOverrideAngularDamping=true
    numSolverIterationsPostUpdate=4
    numSolverIterationsPreUpdate=16
    alphaBoolBlend {blendOption=Linear}
    boundBone {boneName=crowbar}
    componentLinearAccScale=<0.8 0.8 0.8>
    componentLinearVelScale=<0.8 0.8 0.8>
    physicsBodyDefinitions [{
      sphereCollisionRadius=0
      boundBone {boneName=crowbar}
      boxExtents=<8 12 12>
      constraintSetup {angularLimitsMax=<5 20 20>, angularLimitsMin=<-5 -20 -20>, angularTarget=<1 0 0>}
      localJointOffset=<-1.5 0 0>
     }]
   }
   n11 = Anim Dynamics - Bone: gem(ComponentPose=@n10, Alpha=0.325, ExternalForce="0,0,0") {
    angularDampingOverride=0.99
    angularSpringConstant=500
    bAngularSpring=true
    bGravityOverrideInSimSpace=true
    bOverrideAngularDamping=true
    numSolverIterationsPostUpdate=4
    numSolverIterationsPreUpdate=16
    alphaBoolBlend {blendOption=Linear}
    boundBone {boneName=gem}
    componentLinearAccScale=<0.8 0.8 0.8>
    componentLinearVelScale=<0.8 0.8 0.8>
    physicsBodyDefinitions [{
      sphereCollisionRadius=0
      boundBone {boneName=gem}
      boxExtents=<15 25 25>
      constraintSetup {angularLimitsMax=<5 5 60>, angularLimitsMin=<-5 -5 -60>, angularTarget=<1 0 0>}
      localJointOffset=<-2 0 0>
     }]
   }
   n12 = Anim Dynamics - Bone: crystal(ComponentPose=@n11, Alpha=0.4, ExternalForce="0,0,0") {
    angularDampingOverride=0.99
    angularSpringConstant=500
    bAngularSpring=true
    bGravityOverrideInSimSpace=true
    bOverrideAngularDamping=true
    numSolverIterationsPostUpdate=4
    numSolverIterationsPreUpdate=16
    alphaBoolBlend {blendOption=Linear}
    boundBone {boneName=crystal}
    componentLinearAccScale=<0.8 0.8 0.8>
    componentLinearVelScale=<0.8 0.8 0.8>
    physicsBodyDefinitions [{
      sphereCollisionRadius=0
      boundBone {boneName=crystal}
      boxExtents=<13 19 19>
      constraintSetup {angularLimitsMax=<5 20 20>, angularLimitsMin=<-5 -20 -20>, angularTarget=<1 0 0>}
      localJointOffset=<-2.5 0 0>
     }]
   }
   n13 = Anim Dynamics - Bone: flap_r(ComponentPose=@n12, Alpha=1, ExternalForce="0,0,0") {
    angularDampingOverride=0.99
    angularSpringConstant=3000
    bAngularSpring=true
    bGravityOverrideInSimSpace=true
    bOverrideAngularDamping=true
    numSolverIterationsPostUpdate=4
    numSolverIterationsPreUpdate=16
    alphaBoolBlend {blendOption=Linear}
    boundBone {boneName=flap_r}
    componentLinearAccScale=<0.8 0.8 0.8>
    componentLinearVelScale=<0.8 0.8 0.8>
    physicsBodyDefinitions [{
      sphereCollisionRadius=0
      boundBone {boneName=flap_r}
      boxExtents=<13 19 19>
      constraintSetup {angularLimitsMax=<5 20 5>, angularLimitsMin=<-5 -20 -5>, angularTarget=<1 0 0>}
      localJointOffset=<-2.5 0 0>
     }]
   }
   n14 = Anim Dynamics - Bone: strap_r(ComponentPose=@n13, Alpha=1, ExternalForce="0,0,0") {
    angularDampingOverride=0.99
    angularSpringConstant=3000
    bAngularSpring=true
    bGravityOverrideInSimSpace=true
    bOverrideAngularDamping=true
    numSolverIterationsPostUpdate=4
    numSolverIterationsPreUpdate=16
    alphaBoolBlend {blendOption=Linear}
    boundBone {boneName=strap_r}
    componentLinearAccScale=<0.8 0.8 0.8>
    componentLinearVelScale=<0.8 0.8 0.8>
    physicsBodyDefinitions [{
      sphereCollisionRadius=0
      boundBone {boneName=strap_r}
      boxExtents=<13 19 19>
      constraintSetup {angularLimitsMax=<0 0 20>, angularLimitsMin=<0 0 -20>, angularTarget=<1 0 0>}
      localJointOffset=<-2.5 0 0>
     }]
   }
   n15 = Anim Dynamics - Bone: flap_l(ComponentPose=@n14, Alpha=1, ExternalForce="0,0,0") {
    angularDampingOverride=0.99
    angularSpringConstant=3000
    bAngularSpring=true
    bGravityOverrideInSimSpace=true
    bOverrideAngularDamping=true
    numSolverIterationsPostUpdate=4
    numSolverIterationsPreUpdate=16
    alphaBoolBlend {blendOption=Linear}
    boundBone {boneName=flap_l}
    componentLinearAccScale=<0.8 0.8 0.8>
    componentLinearVelScale=<0.8 0.8 0.8>
    physicsBodyDefinitions [{
      sphereCollisionRadius=0
      boundBone {boneName=flap_l}
      boxExtents=<13 19 19>
      constraintSetup {angularLimitsMax=<5 20 5>, angularLimitsMin=<-5 -20 -5>, angularTarget=<1 0 0>}
      localJointOffset=<-2.5 0 0>
     }]
   }
   n16 = Anim Dynamics - Bone: strap_l(ComponentPose=@n15, Alpha=1, ExternalForce="0,0,0") {
    angularDampingOverride=0.99
    angularSpringConstant=3000
    bAngularSpring=true
    bGravityOverrideInSimSpace=true
    bOverrideAngularDamping=true
    numSolverIterationsPostUpdate=4
    numSolverIterationsPreUpdate=16
    alphaBoolBlend {blendOption=Linear}
    boundBone {boneName=strap_l}
    componentLinearAccScale=<0.8 0.8 0.8>
    componentLinearVelScale=<0.8 0.8 0.8>
    physicsBodyDefinitions [{
      sphereCollisionRadius=0
      boundBone {boneName=strap_l}
      boxExtents=<13 19 19>
      constraintSetup {angularLimitsMax=<0 0 20>, angularLimitsMin=<0 0 -20>, angularTarget=<1 0 0>}
      localJointOffset=<-2.5 0 0>
     }]
   }
   Save cached pose 'GrappelingHookCache'(Pose=Component To Local(ComponentPose=@n16)) {cacheName=GrappelingHookCache}
  ]}
 EventGraph {nodes [Event BlueprintUpdateAnimation()]}
]
variables []
